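# Control Rig for the TechTomato male body's face accessory that procedurally spins its five extra wheel bones.
# A sequence takes the RotRing variable modulo 360, scales it by a different signed multiplier per wheel and applies it as a relative X rotation, giving varied speeds and directions.
# A branch negating neck_01's local Euler rotation targets extra_neck at zero weight, and a Set Rotation on neck_01 is unwired, so neither takes effect.
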ControlRigBlueprint /BRCosmetics/Characters/Player/Male/Medium/Bodies/M_MED_TechTomato/Meshes/Parts/M_MED_TechTomato_FaceAcc_CtrlRig
graphs [
 Multiply_1_1_1_1_ContainedGraph {nodes [
   n1 = RigVMFunctionEntryNode()
   RigVMFunctionReturnNode(Result=&RigVMFunction_MathDoubleMul(B=@n1.C, A=&RigVMFunction_MathDoubleMul(A=@n1.A, B=@n1.B).Result).Result)
  ]}
 Multiply_1_1_2_ContainedGraph {nodes [
   n1 = RigVMFunctionEntryNode()
   RigVMFunctionReturnNode(Result=&RigVMFunction_MathDoubleMul(B=@n1.C, A=&RigVMFunction_MathDoubleMul(A=@n1.A, B=@n1.B).Result).Result)
  ]}
 Multiply_1_2_ContainedGraph {nodes [
   n1 = RigVMFunctionEntryNode()
   RigVMFunctionReturnNode(Result=&RigVMFunction_MathDoubleMul(B=@n1.C, A=&RigVMFunction_MathDoubleMul(A=@n1.A, B=@n1.B).Result).Result)
  ]}
 RigVMFunctionLibrary {links [], nodes []}
 RigVMFunction_Sequence_1_ContainedGraph {nodes [
   n1 = RigVMFunctionEntryNode()
   RigVMFunction_Sequence = &RigVMFunction_Sequence(ExecuteContext=@n1.ExecuteContext)
   n2 = &RigVMFunction_Sequence(ExecuteContext=@RigVMFunction_Sequence.B)
   n3 = &RigVMFunction_Sequence(ExecuteContext=@n2.B)
   n4 = &RigVMFunction_Sequence(ExecuteContext=@n3.B)
   n5 = &RigVMFunction_Sequence(ExecuteContext=@n4.B)
   RigVMFunctionReturnNode(A=@RigVMFunction_Sequence.A, B=@n2.A, C=@n3.A, D=@n4.A, E=@n5.A, F=@n5.B)
  ]}
 RigVMModel {nodes [
   RigUnit_BeginExecution = &RigUnit_BeginExecution()
   RigVMFunction_Sequence = RigVMAggregateNode(ExecuteContext=@RigUnit_BeginExecution.ExecutePin)
   n1 = /Script/RigVM.RigVMFunction_MathDoubleMod(B=360, A=RigVMVariableNode() {variable=RotRing}.Value) {is=struct RigVMUnitNode, template="Modulo::Execute(in A,in B,out Result)"}
   /Script/ControlRig.RigUnit_SetRelativeTransformForItem(
    Child=(Type=Bone,Name="extra_wheel_r")
    Parent=(Type=Bone,Name="extra_wheel_r")
    Value=(Rotation=(X=0,Y=0,Z=0,W=1),Translation=(X=0,Y=0,Z=0),Scale3D=(X=1,Y=1,Z=1))
    Weight=1
    bParentInitial=False
    bPropagateToChildren=True
    Value.Rotation=&RigVMFunction_MathQuaternionFromEuler(Euler=(X=0,Y=0,Z=0), RotationOrder=XYZ, Euler.X=&RigVMFunction_MathDoubleMul(B=-200, A=@n1.Result).Result).Result
    ExecutePin=@RigVMFunction_Sequence.A
   ) {
    is=struct RigVMUnitNode
    template="Set Relative Transform::Execute(in Child,in Parent,in bParentInitial,in Value,in Weight,in bPropagateToChildren)"
   }
   n2 = &RigVMFunction_MathDoubleMod(B=360, A=RigVMVariableNode() {variable=RotRing}.Value) {template="Modulo::Execute(in A,in B,out Result)"}
   &RigUnit_SetRelativeTransformForItem(
    Child=(Type=Bone,Name="extra_wheel_l")
    Parent=(Type=Bone,Name="extra_wheel_l")
    Value=(Rotation=(X=0,Y=0,Z=0,W=1),Translation=(X=0,Y=0,Z=0),Scale3D=(X=1,Y=1,Z=1))
    Weight=1
    bParentInitial=False
    bPropagateToChildren=True
    Value.Rotation=&RigVMFunction_MathQuaternionFromEuler(Euler=(X=0,Y=0,Z=0), RotationOrder=XYZ, Euler.X=&RigVMFunction_MathDoubleMul(B=35, A=@n2.Result).Result).Result
    ExecutePin=@RigVMFunction_Sequence.B
   ) {template="Set Relative Transform::Execute(in Child,in Parent,in bParentInitial,in Value,in Weight,in bPropagateToChildren)"}
   n3 = &RigVMFunction_MathDoubleMod(B=360, A=RigVMVariableNode() {variable=RotRing}.Value) {template="Modulo::Execute(in A,in B,out Result)"}
   &RigUnit_SetRelativeTransformForItem(
    Child=(Type=Bone,Name="extra_wheel_c")
    Parent=(Type=Bone,Name="extra_wheel_c")
    Value=(Rotation=(X=0,Y=0,Z=0,W=1),Translation=(X=0,Y=0,Z=0),Scale3D=(X=1,Y=1,Z=1))
    Weight=1
    bParentInitial=False
    bPropagateToChildren=True
    Value.Rotation=&RigVMFunction_MathQuaternionFromEuler(Euler=(X=0,Y=0,Z=0), RotationOrder=XYZ, Euler.X=&RigVMFunction_MathDoubleMul(B=-90, A=@n3.Result).Result).Result
    ExecutePin=@RigVMFunction_Sequence.C
   ) {template="Set Relative Transform::Execute(in Child,in Parent,in bParentInitial,in Value,in Weight,in bPropagateToChildren)"}
   n4 = &RigVMFunction_MathQuaternionToEuler(
    RotationOrder=XYZ
    Value=&RigUnit_GetTransform(Item=(Type=Bone,Name="neck_01"), Space=LocalSpace, bInitial=false).Transform.Rotation
   )
   Multiply_1 = RigVMAggregateNode(B=-1, C=1, A=@n4.Result.X)
   Multiply_1_1 = RigVMAggregateNode(B=-1, C=1, A=@n4.Result.Y)
   Multiply_1_1_1 = RigVMAggregateNode(B=-1, C=1, A=@n4.Result.Z)
   n5 = &RigVMFunction_MathQuaternionFromEuler(Euler=(X=0,Y=0,Z=90), RotationOrder=XYZ, Euler.X=@Multiply_1.Result, Euler.Y=@Multiply_1_1.Result, Euler.Z=@Multiply_1_1_1.Result)
   &RigUnit_SetRelativeTransformForItem(
    Child=(Type=Bone,Name="extra_neck")
    Parent=(Type=Bone,Name="extra_neck")
    Value=(Rotation=(X=0,Y=0,Z=0,W=1),Translation=(X=0,Y=0,Z=0),Scale3D=(X=1,Y=1,Z=1))
    Weight=0
    bParentInitial=false
    bPropagateToChildren=True
    Value.Rotation=@n5.Result
    ExecutePin=@RigVMFunction_Sequence.F
   ) {template="Set Relative Transform::Execute(in Child,in Parent,in bParentInitial,in Value,in Weight,in bPropagateToChildren)"}
   n6 = &RigVMFunction_MathDoubleMod(B=360, A=RigVMVariableNode() {variable=RotRing}.Value) {template="Modulo::Execute(in A,in B,out Result)"}
   &RigUnit_SetRelativeTransformForItem(
    Child=(Type=Bone,Name="extra_wheel_cf")
    Parent=(Type=Bone,Name="extra_wheel_cf")
    Value=(Rotation=(X=0,Y=0,Z=0,W=1),Translation=(X=0,Y=0,Z=0),Scale3D=(X=1,Y=1,Z=1))
    Weight=1
    bParentInitial=False
    bPropagateToChildren=True
    Value.Rotation=&RigVMFunction_MathQuaternionFromEuler(Euler=(X=0,Y=0,Z=0), RotationOrder=XYZ, Euler.X=&RigVMFunction_MathDoubleMul(B=400, A=@n6.Result).Result).Result
    ExecutePin=@RigVMFunction_Sequence.D
   ) {template="Set Relative Transform::Execute(in Child,in Parent,in bParentInitial,in Value,in Weight,in bPropagateToChildren)"}
   n7 = &RigVMFunction_MathDoubleMod(B=360, A=RigVMVariableNode() {variable=RotRing}.Value) {template="Modulo::Execute(in A,in B,out Result)"}
   &RigUnit_SetRelativeTransformForItem(
    Child=(Type=Bone,Name="extra_wheel_cb")
    Parent=(Type=Bone,Name="extra_wheel_cb")
    Value=(Rotation=(X=0,Y=0,Z=0,W=1),Translation=(X=0,Y=0,Z=0),Scale3D=(X=1,Y=1,Z=1))
    Weight=1
    bParentInitial=False
    bPropagateToChildren=True
    Value.Rotation=&RigVMFunction_MathQuaternionFromEuler(Euler=(X=0,Y=0,Z=0), RotationOrder=XYZ, Euler.X=&RigVMFunction_MathDoubleMul(B=-400, A=@n7.Result).Result).Result
    ExecutePin=@RigVMFunction_Sequence.E
   ) {template="Set Relative Transform::Execute(in Child,in Parent,in bParentInitial,in Value,in Weight,in bPropagateToChildren)"}
   Set Transform = &RigUnit_SetRotation(ExecutePin=(), Item=(Type=Bone,Name="neck_01"), Weight=1, bInitial=False, bPropagateToChildren=true, Value=@n5.Result)
  ]}
]
hierarchy [
 root {loc=<0 0 0>, rot=<0 0 0 1>}
 pelvis {parent=root, loc=<0 -0.7071 93.1191>, rot=<0.0127 0.707 -0.0127 -0.707>}
 spine_01 {parent=pelvis, loc=<6.7494 -0.6388 0>, rot=<0 0 0.034 -0.9994>}
 spine_02 {parent=spine_01, loc=<7.2518 -0.035 0>, rot=<0 0 0.0238 -0.9997>}
 spine_03 {parent=spine_02, loc=<6.0104 0 0>, rot=<0 0 0.0149 -0.9999>}
 spine_04 {parent=spine_03, loc=<8.1024 0 0>, rot=<0 0 -0.0132 -0.9999>}
 spine_05 {parent=spine_04, loc=<10.0146 0 0>, rot=<0 0 -0.0015 -1>}
 extra_neck {parent=spine_05, loc=<11.3138 0.1529 0>, rot=<0 0 -0.2003 -0.9797>}
 extra_wheel_r {parent=extra_neck, loc=<11.2173 -0.5109 7.7244>, rot=<-0.5546 -0.4386 -0.4386 -0.5546>}
 extra_wheel_c {parent=extra_neck, loc=<16.807 1.0682 -0.263>, rot=<-0.5546 -0.4386 -0.4386 -0.5546>}
 extra_wheel_l {parent=extra_neck, loc=<11.6951 -2.294 -7.839>, rot=<-0.5542 -0.4392 -0.4392 -0.5542>}
 extra_wheel_cf {parent=extra_neck, loc=<7.3974 3.9372 -0.2565>, rot=<-0.5577 -0.4347 -0.4347 -0.5577>}
 extra_wheel_cb {parent=extra_neck, loc=<5.6681 -3.2359 0.0082>, rot=<0.4458 -0.5488 0.5565 -0.4363>}
 neck_01 {parent=spine_05, loc=<11.3138 0.1529 0>, rot=<0 0 -0.2003 -0.9797>}
 neck_02 {parent=neck_01, loc=<4.5627 0 0>, rot=<0 0 0.0504 -0.9987>}
 head {parent=neck_02, loc=<5.3914 0 0>, rot=<0 0 0.115 -0.9934>}
]
variables [RotRing {type=double, default=0}]
