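# Animation Blueprint for the TechTomato character's face accessory mesh part, driving its procedural motion through a Control Rig.
# It copies the pose from the attached parent mesh, then passes it through the accessory's Control Rig to the output.
# Each update stores real-time seconds in Time, which feeds the rig's RotRing input, presumably for continuous ring rotation.
# Five gear rotation variables are declared but unused in these graphs.
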AnimBlueprint /BRCosmetics/Characters/Player/Male/Medium/Bodies/M_MED_TechTomato/Meshes/Parts/M_MED_TechTomato_FaceAcc_AnimBP
target_skeleton=&/M_MED_TechTomato_FaceAcc_Skeleton
graphs [
 AnimGraph {nodes [
   n1 = Control Rig!(Source=Copy Pose From Mesh() {bUseAttachedParent=true}, Alpha, RotRing:double=Get Time() -> (Time:double)) -> (Pose) {
    alphaBoolBlend {blendOption=Linear}
    controlRigAssetReference {blueprintRigClass=&/M_MED_TechTomato_FaceAcc_CtrlRig.M_MED_TechTomato_FaceAcc_CtrlRig_C}
   }
   Output Pose(Result=@n1) {layerGroup=None}
  ]}
 EventGraph {nodes [
   Event BlueprintUpdateAnimation()
   > Set Time(Time:double=GetRealTimeSeconds()) -> (Output_Get:double)
  ]}
]
variables [
 Gear_c_rotation {type=Rotator}
 Gear_cb_rotation {type=Rotator}
 Gear_cf_rotation {type=Rotator}
 Gear_l_rotation {type=Rotator}
 Gear_r_rotation {type=Rotator}
 Time {type=number, default=86.4933}
]
